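# PilgrimQuickplay HUD indicator layer widget, a child of the Athena indicator layer that adds no widgets or variables of its own.
# On Construct it runs the parent logic, then sets name indicators to draw their elements in order.
# It then sets the GeneralIndicators actor canvas to grid slot layer 2, adjusting its stacking within the layer.
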WidgetBlueprint /PilgrimQuickplay/Widgets/HUD/PilgrimQuickplay_IndicatorLayer
graphs [EventGraph {nodes [
   Event Construct()
   > Parent: Construct()
   > NameIndicators_SetDrawElementsInOrder!(execute, self:Athena Indicator Layer Base Object Reference, bDrawInOrder:bool=true) -> (then)
   > SetGridSlotLayer!(execute, Widget:Widget Object Reference=Get GeneralIndicators() -> (GeneralIndicators:Fort Actor Canvas Object Reference), Layer:int=2) -> (then)
  ]}]
variables []
widgets {
 info {inheritedWidgetCount=0, parentClass=/Game/Athena/HUD/AthenaIndicatorLayer.AthenaIndicatorLayer_C, rootWidgetClass=null, widgetCount=0}
 widgets []
}
