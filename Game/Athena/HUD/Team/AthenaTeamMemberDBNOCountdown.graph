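# Team HUD widget that shows a small countdown indicator for a downed (DBNO) team member.
# A fixed-size square box holds a visibility switcher whose only panel overlays a centered elimination icon with a radial-progress material image filling the square.
# The event graph is empty; the timer image is inherited from the native AthenaTeamMemberDBNOCountdown parent, so this Blueprint supplies only layout and brushes.
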WidgetBlueprint /Game/Athena/HUD/Team/AthenaTeamMemberDBNOCountdown
graphs [EventGraph {links [], nodes []}]
variables []
widgets {
 info {inheritedWidgetCount=1, parentClass=/Script/FortniteUI.AthenaTeamMemberDBNOCountdown, rootWidgetClass=&SizeBox, widgetCount=5}
 widgets [
  SizeBox_1 {
   widgetClassPath=&SizeBox
   widget=SizeBox_1
   properties {bOverride_HeightOverride=true, bOverride_WidthOverride=true, heightOverride=42, widthOverride=42}
  }
  Switcher_Indicator {widgetClassPath=&CommonVisibilitySwitcher, parent=SizeBox_1, widget=Switcher_Indicator, slot {class=SizeBoxSlot}}
  Panel_DefaultIndicator {widgetClassPath=&Overlay, parent=Switcher_Indicator, widget=Panel_DefaultIndicator, slot {class=CommonVisibilitySwitcherSlot}}
  LazyImage_Indicator {
   widgetClassPath=&FortLazyImage
   parent=Panel_DefaultIndicator
   widget=LazyImage_Indicator
   properties {brush {
     bIsDynamicallyLoaded=false
     drawAs=Image
     imageType=NoImage
     mirroring=NoMirror
     resourceName=None
     resourceObject=/Game/Athena/HUD/Materials/T_UI_ElimIndicator_Flat
     tiling=NoTile
     imageSize=<24 24>
     margin=<ltrb 0 0 0 0>
     outlineSettings {
      bUseBrushTransparency=false
      roundingType=HalfHeightRadius
      width=0
      color {colorUseRule=UseColor_Specified, specifiedColor=<rgba 0 0 0 0>}
      cornerRadii=<0 0 0 0>
     }
     tintColor {colorUseRule=UseColor_Specified, specifiedColor=<rgba 1 1 1 1>}
     uVRegion {bIsValid=false, max=<0 0>, min=<0 0>}
    }}
   slot {class=OverlaySlot, properties {horizontalAlignment=HAlign_Center, verticalAlignment=VAlign_Center}}
  }
  Image_Timer {
   widgetClassPath=&Image
   bInherited=true
   parent=Panel_DefaultIndicator
   widget=Image_Timer
   properties {
    visibility=SelfHitTestInvisible
    brush {
     bIsDynamicallyLoaded=false
     drawAs=Image
     imageType=NoImage
     mirroring=NoMirror
     resourceName=None
     resourceObject=/Game/Athena/HUD/Materials/M_UI_RadialProgress_Countdown
     tiling=NoTile
     imageSize=<42 42>
     margin=<ltrb 0 0 0 0>
     outlineSettings {
      bUseBrushTransparency=false
      roundingType=HalfHeightRadius
      width=0
      color {colorUseRule=UseColor_Specified, specifiedColor=<rgba 0 0 0 0>}
      cornerRadii=<0 0 0 0>
     }
     tintColor {colorUseRule=UseColor_Specified, specifiedColor=<rgba 1 1 1 1>}
     uVRegion {bIsValid=false, max=<0 0>, min=<0 0>}
    }
   }
   slot {class=OverlaySlot, properties {horizontalAlignment=HAlign_Fill, verticalAlignment=VAlign_Fill}}
  }
 ]
}
