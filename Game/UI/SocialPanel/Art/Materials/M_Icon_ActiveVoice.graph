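# Translucent UI material for a social panel voice-chat icon, showing active, muted, self-muted and parental-restricted states.
# A speaker glyph, or a mic when self-muted, combines with three sound-wave channels from one texture, animated by Fmod time steps over a base alpha.
# Mute parameters replace the waves with an offset muted glyph.
# IsRestricted adds a restricted glyph and feeds a TurnEverythingOff named reroute before BaseOpacity scales the sum.
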Material /Game/UI/SocialPanel/Art/Materials/M_Icon_ActiveVoice
graph {
 properties {
  bUsedWithStaticMesh=false
  blendMode=BLEND_Translucent
  materialDomain=MD_UI
  numCustomizedUVs=3
  pixelDepthOffsetMode=PDOM_Legacy
  usedShadingModels=Default Lit
 }
 nodes [
  n1 = Multiply(A=TextureCoordinate(), B=AppendVector(A=Constant() {r=1.5}, B=Constant() {r=1}))
  Add(A=@n1, B=AppendVector(A=Constant() {r=-0.5}, B=Constant())) {constB=0.25}
  Constant() {r=1}
  n2 = TextureCoordinate() {coordinateIndex=1}
  n3 = LinearInterpolate(
   A=TextureSample(UVs=@n2) {samplerType=SAMPLERTYPE_Masks, texture=/Game/UI/Foundation/Textures/Icons/Audio/T_UI_Icon_VoiceSpeaker}.R
   B=TextureSample(UVs=@n2) {texture=/Game/UI/Foundation/Textures/Icons/Audio/T_UI_Icon_VoiceMic}.A
   Alpha=Multiply(A=ScalarParameter() {parameterName=IsSelfMuted}, B=NamedRerouteUsage())
  )
  n4 = TextureCoordinate()
  n5 = AppendVector(A=Constant() {r=0.75}, B=Constant() {r=0.75})
  n6 = Constant() {r=-0.175}
  n7 = Constant()
  n8 = TextureSample(UVs=Add(
    A=MaterialFunctionCall(UVs=@n4, Texture Scale=@n5) {materialFunction=&ScaleUVsByCenter}.UVs
    B=AppendVector(A=@n6, B=@n7)
   )) {texture=/Game/UI/Foundation/Textures/Icons/Audio/T_UI_Icon_VoiceMuted}
  n9 = Multiply(A=NamedRerouteUsage(), B=ScalarParameter() {defaultValue=1, parameterName=AnimateStart, sliderMax=1})
  n10 = Constant()
  n11 = TextureSampleParameter2D(UVs=TextureCoordinate()) {parameterName=SoundTexture, samplerType=SAMPLERTYPE_Masks, texture=/Game/UI/Foundation/Textures/Icons/Audio/T_UI_Icon_VoiceNoise}
  n12 = If(A=Constant() {r=0.25}, B=@n9, "A > B"=@n10, "A == B"=@n11.B, "A < B"=@n11.B)
  n13 = Add(B=@n12)
  n14 = If(A=Constant() {r=0.5}, B=@n9, "A > B"=@n10, "A == B"=@n11.G, "A < B"=@n11.G)
  n15 = If(A=Constant() {r=0.75}, B=@n9, "A > B"=@n10, "A == B"=@n11.R, "A < B"=@n11.R)
  n16 = Multiply(A=Add(A=Add(A=@n13, B=@n14), B=@n15), B=ScalarParameter() {defaultValue=0.33, parameterName=BaseAlpha})
  n17 = Multiply(A=Time(), B=Multiply(A=ScalarParameter() {defaultValue=1, parameterName=AnimSpeed}, B=NamedRerouteUsage()))
  n18 = Constant() {r=0.5}
  n19 = Constant() {r=1}
  n20 = Constant()
  n21 = If(A=Fmod(A=@n17, B=Constant() {r=1}), B=@n18, "A > B"=@n19, "A == B"=@n20, "A < B"=@n20)
  n22 = If(A=Fmod(A=Divide(A=@n17), B=Constant() {r=1}), B=@n18, "A > B"=@n19, "A == B"=@n20, "A < B"=@n20)
  n23 = OneMinus(None=@n22)
  n24 = LinearInterpolate(A=OneMinus(None=@n21), Alpha=@n23) {constB=0}
  n25 = LinearInterpolate(A=@n21, Alpha=@n22) {constB=0}
  n26 = Add(
   A=LinearInterpolate(B=@n15, Alpha=LinearInterpolate(A=@n21, Alpha=@n23) {constB=0}) {constB=0}
   B=Add(A=LinearInterpolate(B=@n14, Alpha=@n24) {constB=0}, B=LinearInterpolate(B=@n12, Alpha=@n25) {constB=0})
  )
  n27 = ScalarParameter() {parameterName=IsMuted}
  n28 = NamedRerouteUsage()
  n29 = LinearInterpolate(
   A=LinearInterpolate(A=Add(A=@n16, B=@n26), B=@n8.A, Alpha=Multiply(A=@n27, B=@n28))
   B=@n8.A
   Alpha=Multiply(A=ScalarParameter() {parameterName=IsSelfMuted}, B=NamedRerouteUsage())
  )
  n30 = TextureCoordinate()
  n31 = AppendVector(A=Constant() {r=0.75}, B=Constant() {r=0.75})
  n32 = Constant() {r=-0.175}
  n33 = Constant()
  n34 = TextureSample(UVs=Add(
    A=MaterialFunctionCall(UVs=@n30, Texture Scale=@n31) {materialFunction=&ScaleUVsByCenter}.UVs
    B=AppendVector(A=@n32, B=@n33)
   )) {texture=/Game/UI/Foundation/Textures/Icons/Audio/T_UI_Icon_VoiceParentalRestricted}
  n35 = ScalarParameter() {parameterName=IsRestricted}
  Multiply(A=Add(A=Add(A=@n3, B=@n29), B=Multiply(A=@n34.A, B=@n35)), B=ScalarParameter() {defaultValue=1, parameterName=BaseOpacity})
  n36 = ScalarParameter() {parameterName=IsRestricted}
  NamedRerouteDeclaration(None=Floor(None=OneMinus(None=@n36))) {name=TurnEverythingOff, nodeColor=<rgba 0 0 0 1>}
  Reroute(None=@n1)
 ]
}
