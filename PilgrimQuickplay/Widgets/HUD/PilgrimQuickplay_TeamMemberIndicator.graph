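# Pilgrim Quickplay HUD nameplate for a teammate, built on AthenaPlayerIndicatorBase.
# A switcher shows an identity pill for local-multiplayer players or an online name with pin-colored marker and voice icon; talking also drives a font material parameter.
# Player state feeds the DBNO widgets, though their animation panel and countdown are then collapsed.
# The widget collapses when its parts are hidden, and screen priority derives from splitscreen or squad index.
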WidgetBlueprint /PilgrimQuickplay/Widgets/HUD/PilgrimQuickplay_TeamMemberIndicator
graphs [
 EventGraph {nodes [
   Update Visibility() {is=title K2Node_CustomEvent}
   > Branch(Condition=Get bShouldUpdateWidgetVisibility() -> (bShouldUpdateWidgetVisibility:bool))
   n1 = IsVisible(self=Get CommonWidgetSwitcher_Marker() -> (CommonWidgetSwitcher_Marker:Common Widget Switcher Legacy Object Reference))
   n2 = IsVisible(self=Get CommonVisibilitySwitcher_PlayerID() -> (CommonVisibilitySwitcher_PlayerID:Common Visibility Switcher Object Reference))
   n3 = IsVisible(self=Get DBNOCountdown() -> (DBNOCountdown:Athena Team Member DBNOCountdown Object Reference))
   n4 = IsVisible(self=Get DBNOIndicator() -> (DBNOIndicator:Athena Team Member DBNOState Object Reference))
   > SetVisibility(InVisibility=Select(Option 0=Collapsed, Option 1=HitTestInvisible, Index=OR Boolean!(A=@n1, B=@n2, C:bool=@n4, D:bool=@n3) -> (ReturnValue)))
   n5 = Event On Handle Player State Set!() -> (OutputDelegate, then, PlayerStateAthena:Fort Player State Athena Object Reference) {is=title K2Node_Event}
   > SetPlayerState!(
    execute
    self:Athena Team Member DBNOState Object Reference=Get DBNOIndicator()
    InPlayerState:Fort Player State Athena Object Reference=@n5.PlayerStateAthena
   ) -> (then)
   > SetPlayerState(self:Athena Team Member DBNOCountdown Object Reference=Get DBNOCountdown(), InPlayerState=@n5.PlayerStateAthena)
   > Update Visibility() {is=title K2Node_CallFunction}
   > Set MyPlayerState(MyPlayerState:Player State Object Reference=@n5.PlayerStateAthena) -> (Output_Get:Player State Object Reference)
   > ShowLocalOrOnlineIdentity!(execute, self, IsLocalMPView:bool=IsHamSandwichSessionActive!() -> (ReturnValue:bool)) -> (then)
   > SetVisibility(
    self=[Get AnimationsPanel_DBNO() -> (AnimationsPanel_DBNO:Fort Simple Widget Animations Panel Object Reference), Get DBNOCountdown()]
    InVisibility=Collapsed
   )
   n6 = Event On Handle Pin Color Changed!() -> (OutputDelegate, then, PinColor:"Linear Color Structure (by ref)") {is=title K2Node_Event}
   > SetBrushTintColor!(
    execute
    self:Image Object Reference=[Get Speech() -> (Speech:Fort Lazy Image Object Reference), Get SquadMarker() -> (SquadMarker:Image Object Reference)]
    TintColor:Slate Color Structure=Make Slate Color(SpecifiedColor=@n6.PinColor)
   ) -> (then)
   n7 = Event On Handle Player Name Replicated!() -> (OutputDelegate, then, PlayerName:String) {is=title K2Node_Event}
   > "SetText (Text)"(self=Get PlayerName() -> (PlayerName:Text Object Reference), InText="To Text (String)"(InString=@n7.PlayerName))
   n8 = Event Update Element Visibility!() -> (OutputDelegate, then, bDisplayWorldName:bool, bDisplayVoiceIndicator:bool, bDisplayBubbleChatIndicator:bool) {is=title K2Node_Event}
   > SetVisibility(
    self=Get CommonWidgetSwitcher_Marker()
    InVisibility=Select(Option 0=Collapsed, Option 1=SelfHitTestInvisible, Index=OR Boolean!(A=@n8.bDisplayWorldName, B=@n8.bDisplayVoiceIndicator) -> (ReturnValue))
   )
   > SetVisibility(
    self=Get CommonVisibilitySwitcher_PlayerID()
    InVisibility=Select(Option 0=Collapsed, Option 1=SelfHitTestInvisible, Index=@n8.bDisplayWorldName)
   )
   n9 = NOT Boolean(A=@n8.bDisplayWorldName)
   > SetVisibility(
    self=Get SquadMarker()
    InVisibility=Select(Option 0=SelfHitTestInvisible, Option 1=Collapsed, Index=AND Boolean(A=@n9, B=@n8.bDisplayVoiceIndicator))
   )
   > SetVisibility(self=Get Speech(), InVisibility=Select(Option 0=Collapsed, Option 1=SelfHitTestInvisible, Index=@n8.bDisplayVoiceIndicator))
   > Update Visibility()
   n10 = Event OnExtensionWidgetCreated!() -> (OutputDelegate, then, Widget:User Widget Object Reference, SlotType:EUIExtensionSlot Enum=Invalid) {is=title K2Node_Event}
   > AddChildToOverlay!(
    execute
    self:Overlay Object Reference=Get OverlayAbove() -> (OverlayAbove:Overlay Object Reference)
    Content:Widget Object Reference=@n10.Widget
   ) -> (then, ReturnValue:Overlay Slot Object Reference)
   n11 = Event On Handle DBNO State Changed!() -> (OutputDelegate, then, bIsDBNO:bool) {is=title K2Node_Event}
   > Branch(Condition=@n11.bIsDBNO)
   > PlayAnimationByName!(
    execute
    self:Fort Simple Widget Animations Panel Object Reference=Get AnimationsPanel_DBNO()
    Name:"Name (by ref)"=MakeLiteralName(Value=DBNOIntro)
    Speed:float=1
    StartAtTime:float=0
   ) -> (then, ReturnValue:bool)
   n12 = Event On Handle Talking State Changed!() -> (OutputDelegate, then, bIsTalking:bool) {is=title K2Node_Event}
   > n13 = GetDynamicFontMaterial!(execute, self:Text Object Reference=Get PlayerName()) -> (then, ReturnValue:Material Instance Dynamic Object Reference)
   > SetScalarParameterValue(self=@n13, ParameterName=IsTalking, Value=SelectFloat(A=1, B=0, bPickA=@n12.bIsTalking))
   > Branch(Condition=IsVisible(self=Get Speech()))
   > SetActiveWidget(
    self:Widget Switcher Object Reference=Get CommonWidgetSwitcher_Marker()
    Widget=Select(Option 0:Widget Object Reference=Get SquadMarker(), Option 1:Widget Object Reference=Get Speech(), Index=@n12.bIsTalking) -> (ReturnValue:Widget Object Reference)
   )
  ]}
 SetLocalOrOnlineScreenPriority {nodes [
   n1 = SetLocalOrOnlineScreenPriority!() -> (then, FortPlayerState:Fort Player State Athena Object Reference) {is=title K2Node_FunctionEntry}
   n2 = int + int(
    A=GetSplitscreenIndex!(PC:Player Controller Object Reference=GetPlayerController!(self:Player State Object Reference=@n1) -> (ReturnValue:Player Controller Object Reference)) -> (ReturnValue:int=0)
    B=10
   )
   n3 = int + int(A=GetSquadMemberIndex!(self:Fort Player State Athena Object Reference=@n1) -> (ReturnValue:int=0), B=14)
   > Set ScreenPriority(ScreenPriority:int=SelectInt(
     A=int * int(A=@n2, B=-1)
     B=int * int(A=@n3, B=-1)
     bPickA=Is Local Player in Local MP Session!(PlayerState:Player State Object Reference=@n1) -> (ReturnValue:bool)
    )) -> (Output_Get:int=0)
  ]}
 ShowLocalOrOnlineIdentity {nodes [
   ShowLocalOrOnlineIdentity!() -> (then, IsLocalMPView:bool=~) {is=title K2Node_FunctionEntry}
   n1 = Get MyPlayerState() -> (MyPlayerState:Player State Object Reference)
   > n2 = Is Valid(InputObject=@n1)
   > n3 = Branch(Condition=Is Local Player in Local MP Session(PlayerState=@n1))
   n4 = Get CommonVisibilitySwitcher_PlayerID()
   > SetActiveWidgetIndex(self:Common Visibility Switcher Object Reference=@n4, Index=0)
   > OverridePlayerInfo!(
    execute
    self:WBP Pilgrim Player Identity Pill Object Reference=Get PlayerIdentity_Pill() -> (PlayerIdentity_Pill:WBP Pilgrim Player Identity Pill Object Reference)
    PlayerState:Player State Object Reference=@n1
   ) -> (then)
   SetActiveWidgetIndex(execute=[@n3.else, @n2.Is Not Valid], self=@n4, Index=1)
  ]}
]
variables [
 bShouldUpdateWidgetVisibility {type=boolean, default=true}
 MyPlayerState {type=PlayerState, default=null}
]
widgets {
 info {inheritedWidgetCount=1, parentClass=/Script/FortniteUI.AthenaPlayerIndicatorBase, rootWidgetClass=&Border, widgetCount=15}
 widgets [
  Border_Nameplate {
   widgetClassPath=&Border
   widget=Border_Nameplate
   properties {
    bIsVolatile=true
    horizontalAlignment=HAlign_Center
    verticalAlignment=VAlign_Center
    visibility=SelfHitTestInvisible
    background {
     bIsDynamicallyLoaded=false
     drawAs=NoDrawType
     imageType=NoImage
     mirroring=NoMirror
     resourceName=None
     resourceObject=/Game/Athena/HUD/Team/M_UI_Pill_SquadNamePlate_Inst
     tiling=NoTile
     imageSize=<32 32>
     margin=<ltrb 0.5 0 0.5 0>
     outlineSettings {
      bUseBrushTransparency=false
      roundingType=HalfHeightRadius
      width=0
      color {colorUseRule=UseColor_Specified, specifiedColor=<rgba 0 0 0 0>}
      cornerRadii=<0 0 0 0>
     }
     tintColor {colorUseRule=UseColor_Specified, specifiedColor=<rgba 1 1 1 1>}
     uVRegion {bIsValid=false, max=<0 0>, min=<0 0>}
    }
    brushColor=<rgba 0 0 0 0.33>
    padding=<ltrb 10 0 10 0>
   }
  }
  Overlay_2 {
   widgetClassPath=&Overlay
   parent=Border_Nameplate
   widget=Overlay_2
   properties {bIsVolatile=true}
   slot {class=BorderSlot, properties {horizontalAlignment=HAlign_Center, verticalAlignment=VAlign_Center, padding=<ltrb 10 0 10 0>}}
  }
  CommonVisibilitySwitcher_PlayerID {
   widgetClassPath=&CommonVisibilitySwitcher
   bIsVariable=true
   parent=Overlay_2
   widget=CommonVisibilitySwitcher_PlayerID
   slot {class=OverlaySlot, properties {horizontalAlignment=HAlign_Fill, verticalAlignment=VAlign_Fill}}
  }
  PlayerIdentity_Pill {
   widgetClassPath=/PilgrimCore/Widgets/Player/WBP_Pilgrim_PlayerIdentity_Pill
   bIsVariable=true
   parent=CommonVisibilitySwitcher_PlayerID
   widget=PlayerIdentity_Pill
   slot {class=CommonVisibilitySwitcherSlot}
  }
  HB_OnlineIdentity {
   widgetClassPath=&HorizontalBox
   parent=CommonVisibilitySwitcher_PlayerID
   widget=HB_OnlineIdentity
   slot {class=CommonVisibilitySwitcherSlot}
  }
  CommonWidgetSwitcher_Marker {
   widgetClassPath=&CommonWidgetSwitcherLegacy
   bIsVariable=true
   parent=HB_OnlineIdentity
   widget=CommonWidgetSwitcher_Marker
   properties {bIsVolatile=true, renderTransform {angle=0, scale=<1 1>, shear=<0 0>, translation=<-16 0>}}
   slot {class=HorizontalBoxSlot, properties {horizontalAlignment=HAlign_Left, verticalAlignment=VAlign_Center}}
  }
  SquadMarker {
   widgetClassPath=&Image
   bIsVariable=true
   parent=CommonWidgetSwitcher_Marker
   widget=SquadMarker
   properties {
    bIsVolatile=true
    visibility=SelfHitTestInvisible
    brush {
     bIsDynamicallyLoaded=false
     drawAs=Image
     imageType=NoImage
     mirroring=NoMirror
     resourceName=None
     resourceObject=/Game/Athena/HUD/Art/T_UI_Indicator_Diamond
     tiling=NoTile
     imageSize=<14 16>
     margin=<ltrb 0 0 0 0>
     outlineSettings {
      bUseBrushTransparency=false
      roundingType=HalfHeightRadius
      width=0
      color {colorUseRule=UseColor_Specified, specifiedColor=<rgba 0 0 0 0>}
      cornerRadii=<0 0 0 0>
     }
     tintColor {colorUseRule=UseColor_Specified, specifiedColor=<rgba 1 1 1 1>}
     uVRegion {bIsValid=false, max=<0 0>, min=<0 0>}
    }
    renderTransform {angle=0, scale=<1 1>, shear=<0 0>, translation=<8 0>}
   }
   slot {class=WidgetSwitcherSlot, properties {horizontalAlignment=HAlign_Center, verticalAlignment=VAlign_Center, padding=<ltrb -16 -16 -16 -16>}}
  }
  Speech {
   widgetClassPath=&FortLazyImage
   bIsVariable=true
   parent=CommonWidgetSwitcher_Marker
   widget=Speech
   properties {
    bApplyDropShadow=true
    bIsVolatile=true
    brush {
     bIsDynamicallyLoaded=false
     drawAs=Image
     imageType=NoImage
     mirroring=NoMirror
     resourceName=None
     resourceObject=/Game/UI/SocialPanel/Art/Materials/M_Icon_ActiveVoice
     tiling=NoTile
     imageSize=<32 32>
     margin=<ltrb 0 0 0 0>
     outlineSettings {
      bUseBrushTransparency=false
      roundingType=HalfHeightRadius
      width=0
      color {colorUseRule=UseColor_Specified, specifiedColor=<rgba 0 0 0 0>}
      cornerRadii=<0 0 0 0>
     }
     tintColor {colorUseRule=UseColor_Specified, specifiedColor=<rgba 1 1 1 1>}
     uVRegion {bIsValid=false, max=<0 0>, min=<0 0>}
    }
   }
   slot {class=WidgetSwitcherSlot, properties {horizontalAlignment=HAlign_Center, verticalAlignment=VAlign_Center, padding=<ltrb -16 -16 -16 -16>}}
  }
  InvalidationBox_0 {
   widgetClassPath=/Script/UMG.InvalidationBox
   bIsVariable=true
   parent=HB_OnlineIdentity
   widget=InvalidationBox_0
   properties {bIsVolatile=true}
   slot {class=HorizontalBoxSlot}
  }
  PlayerName {
   widgetClassPath=&TextBlock
   bIsVariable=true
   parent=InvalidationBox_0
   widget=PlayerName
   properties {
    bIsVolatile=true
    text="{MaxCharacterLim}"
    visibility=SelfHitTestInvisible
    font {
     bForceMonospaced=false
     bMaterialIsStencil=true
     fontMaterial=/Game/Athena/HUD/Team/M_Font_PlayerIndicator
     fontObject=&Burbank
     letterSpacing=0
     monospacedWidth=1
     size=14
     skewAmount=0
     typefaceFontName=BurbankRegular-Black
     outlineSettings {
      bApplyOutlineToDropShadows=false
      bMiteredCorners=false
      bSeparateFillAlpha=false
      outlineMaterial=null
      outlineSize=0
      outlineColor=<rgba 0 0 0 1>
     }
    }
    renderTransform {angle=0, scale=<1 1>, shear=<-8 0>, translation=<0 0>}
   }
   slot {class=PanelSlot}
  }
  AnimationsPanel_DBNO {
   widgetClassPath=/Script/FortniteUI.FortSimpleWidgetAnimationsPanel
   bIsVariable=true
   parent=Overlay_2
   widget=AnimationsPanel_DBNO
   properties {
    bIsVolatile=true
    animations [DBNOIntro {
      bIsEnabled=true
      onAnimationsFinished=()
      targets [{
        alpha=/Game/Athena/HUD/Team/AthenaTeamMemberIndicatorDBNOAlpha
        bResetOnFinish=false
        bShouldAnimateAlpha=true
        bShouldAnimateScale=true
        bShouldAnimateTranslation=false
        scale=/Game/Athena/HUD/Team/AthenaTeamMemberIndicatorDBNOScale
        targetName=DBNOIndicator
        translation=null
        materialScalarAnimations []
       }]
     }]
    renderTransform {angle=0, scale=<1 1>, shear=<0 0>, translation=<40 0>}
   }
   slot {class=OverlaySlot, properties {horizontalAlignment=HAlign_Right, verticalAlignment=VAlign_Center, padding=<ltrb 0 -40 0 -40>}}
  }
  DBNOIndicator {
   widgetClassPath=/Game/Athena/HUD/Team/AthenaTeamMemberDBNOState
   bIsVariable=true
   parent=AnimationsPanel_DBNO
   widget=DBNOIndicator
   properties {bIsVolatile=true, visibility=Collapsed}
   slot {class=FortSimpleWidgetAnimationsSlot}
  }
  DBNOCountdown {
   widgetClassPath=/Game/Athena/HUD/Team/AthenaTeamMemberDBNOCountdown
   bIsVariable=true
   parent=Overlay_2
   widget=DBNOCountdown
   properties {bIsVolatile=true, visibility=Collapsed}
   slot {class=OverlaySlot, properties {horizontalAlignment=HAlign_Center, verticalAlignment=VAlign_Center, padding=<ltrb 0 -35 0 0>}}
  }
  OverlayAbove {
   widgetClassPath=&Overlay
   bIsVariable=true
   parent=Overlay_2
   widget=OverlayAbove
   properties {bIsVolatile=true}
   slot {class=OverlaySlot, properties {horizontalAlignment=HAlign_Center, verticalAlignment=VAlign_Center, padding=<ltrb 0 -70 0 0>}}
  }
  TutorialHighlight {widgetClassPath=&TutorialHighlight}
 ]
}
